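# Sound cue that plays bombable wall explosion debris, classed as an explosive sound.
# A root Modulator at default settings feeds a Quality Level switch that picks between two random pools of debris waves.
# The first two quality slots draw from all six variations, while the last two draw from only variations 04 to 06.
# Spatialization uses a debris attenuation asset, and voice limiting references the shared weapon explosion concurrency rule.
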SoundCue /BombableWallsGameplay/Sound/BombableWalls_Explode_Cue
root=Modulator_1
nodes [
 {id=Modulator_1, children [QualityLevel_1]}
 {id=QualityLevel_1, children [Random_2,Random_2,Random_3,Random_3]}
 {
  id=Random_2
  children [WavePlayer_26, WavePlayer_27, WavePlayer_28, WavePlayer_29, WavePlayer_30, WavePlayer_31]
  properties {weights [1,1,1,1,1,1]}
 }
 {id=Random_3, children [WavePlayer_29, WavePlayer_30, WavePlayer_31], properties {bShouldExcludeFromBranchCulling=true, weights [1,1,1]}}
 {id=WavePlayer_26, properties {soundWaveAssetPtr=&/Bombable_Walls_Debris_01}}
 {id=WavePlayer_27, properties {soundWaveAssetPtr=&/Bombable_Walls_Debris_02}}
 {id=WavePlayer_28, properties {soundWaveAssetPtr=&/Bombable_Walls_Debris_03}}
 {id=WavePlayer_29, properties {soundWaveAssetPtr=&/Bombable_Walls_Debris_04}}
 {id=WavePlayer_30, properties {soundWaveAssetPtr=&/Bombable_Walls_Debris_05}}
 {id=WavePlayer_31, properties {soundWaveAssetPtr=&/Bombable_Walls_Debris_06}}
]
properties {
 attenuationSettings=/BombableWallsGameplay/Sound/Settings/Bombable_Walls_Debris_Att
 maxDistance=3100
 soundClassObject=/Game/Sounds/FortSoundClassesAndModes/Fort_Explosives
 volumeMultiplier=1.2
 attenuationOverrides {falloffDistance=3000, attenuationShapeExtents {x=500}}
 concurrencyOverrides {maxCount=4, resolutionRule=StopOldest, volumeScale=0.5}
 concurrencySet [/Game/Sounds/Concurrency_Rules/Weapon_Explosion_Conc]
}
